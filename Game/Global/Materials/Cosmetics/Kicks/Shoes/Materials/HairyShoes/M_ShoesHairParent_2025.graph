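# Parent material for hairy shoe cosmetics, built on the 2023 character parent material functions with material attributes.
# The PreFX function supplies the base material and masks, which named reroutes distribute through the graph.
# Static switches control blending in FN Hair 2023 (on by default) and an optional flyaway hair pass.
# A sock function feeds the final PostFX function.
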Material /Game/Global/Materials/Cosmetics/Kicks/Shoes/Materials/HairyShoes/M_ShoesHairParent_2025
graph {
 properties {
  bUseMaterialAttributes=true
  bUsedWithClothing=true
  bUsedWithMeshDeformer=true
  bUsedWithMorphTargets=true
  bUsedWithNanite=true
  bUsedWithNiagaraMeshParticles=true
  bUsedWithSkeletalMesh=true
  pixelDepthOffsetMode=PDOM_Legacy
  shadingModel=MSM_FromMaterialExpression
  usedShadingModels=Default Lit | Subsurface | Clear Coat | Subsurface Profile | Cloth
 }
 nodes [
  n1 = MaterialFunctionCall() {materialFunction=&MF_CharacterParent_2023_PreFX}
  MaterialFunctionCall(
   Material=MaterialFunctionCall!(InOpacity, Material=NamedRerouteUsage(), Mask=NamedRerouteUsage()) {materialFunction=/Game/Global/Materials/Cosmetics/Kicks/Socks/Materials/MF_Sock}.Result
   M=@n1.M
   Pre Skinned Local Position=@n1.M
   Distance Blend=@n1.M
   Fresnel=@n1.M
  ) {materialFunction=&MF_CharacterParent_2023_PostFX}
  NamedRerouteDeclaration(None=@n1.Result) {name=BaseMaterial, nodeColor=<rgba 0.4531 0 1 1>}
  NamedRerouteDeclaration(None=@n1.M) {name=Mask, nodeColor=<rgba 0 1 0.1328 1>}
  NamedRerouteDeclaration(None=@n1.Tech Art Mask) {name=TechArtMask, nodeColor=<rgba 1 0 0.1875 1>}
  n2 = MaterialFunctionCall!(In=NamedRerouteUsage(), Mask Input=NamedRerouteUsage()) {materialFunction=/Game/Global/Materials/Character_Parent_2023/Hair_Parent_2023/MaterialFunctions/MF_FN_Hair_2023}
  n3 = StaticSwitchParameter(True=BlendMaterialAttributes(A=NamedRerouteUsage(), B=@n2.Result, Alpha=@n2.Result), False=NamedRerouteUsage()) {defaultValue=true, group=Skin FX, parameterName=UseFNHair2_0}
  NamedRerouteDeclaration(None=StaticSwitchParameter(
    True=MaterialFunctionCall!(MaterialInput=@n3) {materialFunction=/Game/Global/Materials/Character_Parent_2023/Hair_Parent_2023/MaterialFunctions/MF_FlyAwayHair_2023}.Result
    False=@n3
   ) {group=Skin FX, parameterName=IsFlyAwayHair?}) {name=HairMaterial, nodeColor=<rgba 0 0.5078 1 1>}
 ]
}
